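# Niagara system for the Gray vehicle boost cosmetic, drawing a warm red-orange exhaust trail with black lightning, sparks and a nozzle flame.
# Two ribbons form the trail and lightning (stepped offset flipbook), changing spawn rate and offset for frontend preview.
# Dots and Particles_Cone pack several sprite and mesh layers into single emitters, varying size, color or placement by spawn group.
# Painted-variant overrides are disabled in every emitter.
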NiagaraSystem /VehicleCosmetics/Boosts/Boost_Gray/FX/NS_Boost_Gray
dependencies {
 usedDataInterfaces [
  &NiagaraDataInterfaceCurve
  &NiagaraDataInterfaceSpriteRendererInfo
  &NiagaraDataInterfaceSkeletalMesh
 ]
 usedDynamicInputs []
 usedModules [
  &self:SystemSpawnScript.SystemScriptSource.SystemScriptGraph.NiagaraNodeAssignment_0.SetVariables_#1
  &SystemState
 ]
}
emitters {
 BlackLightningRibbon001 {
  emitterUpdateScript [
   &EmitterState()
   &VC_FilteredSocketCount()
   &VC_RateQualityHandler()
   &VC_SpawnRibbon_v1(
    Spawn Time=-0.01
    SpawnRate=&MakeCustomFloatFromBool(Bool=$User.bIsFrontendPreview, True Float=60, False Float=30)
   )
   /SetVariables(Emitter.OffsetY=HlslExpression() {hlslExpression="floor(fmod(Engine.Time * 15 , 6)) * .166"}) {isSetParameters=true}
  ]
  particleSpawnScript [
   &VC_PaintedVariantsOverride() {enabled=false}
   &VC_InitializeRibbon_v1(
    Color=<rgba 0 0 0 1>
    Custom Offset=true
    Lifetime=0.336
    Local Vector Offset=&AddVectorToPosition(
     Position=$User.SingleBoostPos
     Vector=/Niagara/DynamicInputs/TypeConversions/MakeCustomPositionFromBool(Bool:NiagaraBool=$User.bIsFrontendPreview, True Position:NiagaraPosition=<-15 0 0>, False Position:NiagaraPosition=<30 0 0>)
    )
    Offset Amount=1
    Ribbon Width=108
   )
   &AddVelocity(Velocity=<-1000 0 0>, Velocity Speed Scale=2) {module=AddVelocity001}
  ]
  particleUpdateScript [
   &ParticleState()
   &InheritSourceMovement(
    Applied Position Scale=&VectorFromFloat(Value=&FloatFromCurve(FloatCurve=DataInterface() {propertyValues {Curve {keys [<at 0 1>,<at 0.5 0>]}}}))
    ApplyForce=false
    ApplyPosition=true
    ApplyPositionFalloff=false
   ) {enabled=false}
   &SolveForcesAndVelocity(Clamp Velocity=false, Limit Acceleration=false)
   &VC_UpdateRibbon_v1(Enable Skeletal Position=true)
   &ScaleRibbonWidth(Ribbon Width Scale=&FloatFromCurve(FloatCurve=DataInterface() {propertyValues {Curve {keys [
         {arriveTangent=2.5201, interpMode=RCIM_Cubic, leaveTangent=2.5201, tangentMode=RCTM_User, time=0, value=1}
         {arriveTangent=-0.5, interpMode=RCIM_Cubic, leaveTangent=-0.5, time=0.15, value=1}
         <at 1 0.5 RCIM_Cubic>
        ]}}})) {enabled=false}
   &DynamicMaterialParameters(
    Index 0 Param 1=1
    Index 1 Param 1=&MaskFloatBySpawnGroup(Masked Float=0, Passthrough Float=1.1)
    Module.Index0Param0Write=true
    Module.Index0Param1Write=true
    Module.Index0Param2Write=true
    Module.Index0Param3Write=true
    Module.Index1Param0Write=true
    Module.Index1Param1Write=true
    Module.Index1Param2Write=true
    Module.Index1Param3Write=true
    Write Parameter Index 1=true
   )
  ]
  rendererClasses [&NiagaraRibbonRendererProperties]
  renderers [{rendererClass=&NiagaraRibbonRendererProperties}]
 }
 Dots {
  emitterUpdateScript [
   &EmitterState()
   &VC_FilteredSocketCount()
   &SpawnRate(SpawnGroup=1, SpawnRate=30, Use Spawn Probability=false)
   &SpawnRate(SpawnGroup=2, SpawnRate=30, Use Spawn Probability=false) {module=SpawnRate001}
  ]
  particleSpawnScript [
   &VC_PaintedVariantsOverride() {enabled=false}
   /SetVariables(Particles.VisTag:NiagaraInt32=&SpawnGroup()) {isSetParameters=true}
   &InitializeParticle(
    Color Mode=::NewEnumerator0
    Lifetime=&RandomRangeFloat(Minimum=0.35, Maximum=0.45)
    Position Offset=&AddVectorToPosition(Position=$System.FinalPosition, Vector=<-75 0 0>)
    Position Offset Coordinate Space=::Local
    Sprite Rotation Mode=::NewEnumerator2
    Sprite Size Mode=::NewEnumerator3
    Uniform Sprite Size=&Multiply_Float(
     A=/VehicleCosmetics/FX/Niagara/Modules/MergeEmitters/VC_FloatBySpawnGroup(
      SpawnGroup or ExecIndex:NiagaraInt32=0
      CustomInt:NiagaraInt32=0
      MaximumSpawnGroups:NiagaraInt32=3
      Float SG 0:NiagaraFloat=0
      Float SG 1:NiagaraFloat=&RandomRangeFloat(Minimum=2, Maximum=0.45)
      Float SG 2:NiagaraFloat=&RandomRangeFloat(Minimum=1.25, Maximum=0.65)
      Float SG 3:NiagaraFloat=0
      Float SG 4:NiagaraFloat=0
      Float SG 5:NiagaraFloat=0
     )
     B=3
    )
    UsePositionOffset=true
    Write Material Random=true
   )
   &AddVelocity(Velocity=&Multiply_Vector(
     A=&RandomRangeVector(Minimum=<-150 0 6>, Maximum=<-35 0 -5>)
     B=&RandomRangeVector(Minimum=<12 0 0>, Maximum=<9 0 0>)
    ))
   &Color(Color=<rgba 0.69 0.42 0.14 1>)
   &ShapeLocation(
    MaskSpawnGroup=false
    Override Seed=false
    Sphere Radius=&MaskFloatBySpawnGroup(Masked Float=0, Passthrough Float=40)
    UseOwnerScale=true
   )
  ]
  particleUpdateScript [
   &ParticleState()
   &ScaleSpriteSize(Uniform Curve Sprite Scale=DataInterface() {propertyValues {Curve {keys [{tangentMode=RCTM_User, time=0, value=1.5}, <at 0.65 1>, {tangentMode=RCTM_User, time=1, value=0}]}}})
   &SolveForcesAndVelocity(Clamp Velocity=false, Limit Acceleration=false)
   &ScaleRibbonWidth(Ribbon Width Scale=&FloatFromCurve(FloatCurve=DataInterface() {propertyValues {Curve {keys [
         {arriveTangent=2.5201, interpMode=RCIM_Cubic, leaveTangent=2.5201, tangentMode=RCTM_User, time=0, value=1}
         {arriveTangent=-0.5, interpMode=RCIM_Cubic, leaveTangent=-0.5, time=0.15, value=1}
         <at 1 0.5 RCIM_Cubic>
        ]}}}))
   &ScaleColor(Scale RGB=&VectorFromFloat(Value=230), ScaleA=true, ScaleRGB=true)
  ]
  rendererClasses [&NiagaraSpriteRendererProperties]
  renderers [
   {rendererClass=&NiagaraSpriteRendererProperties}
   {rendererClass=&NiagaraSpriteRendererProperties, rendererIndex=1}
  ]
 }
 MainTrail001 {
  emitterUpdateScript [
   &EmitterState()
   &VC_FilteredSocketCount()
   &VC_RateQualityHandler()
   &VC_SpawnRibbon_v1(
    Spawn Time=-0.01
    SpawnRate=&MakeCustomFloatFromBool(Bool=$User.bIsFrontendPreview, True Float=60, False Float=30)
   )
  ]
  particleSpawnScript [
   &VC_PaintedVariantsOverride() {enabled=false}
   &VC_InitializeRibbon_v1(
    Color=<rgba 0.87 0.174 0.1859 1>
    Custom Offset=true
    Lifetime=0.7
    Local Vector Offset=&AddVectorToPosition(
     Position=$User.SingleBoostPos
     Vector=&MakeCustomPositionFromBool(Bool:NiagaraBool=$User.bIsFrontendPreview, True Position:NiagaraPosition=<-30 0 0>, False Position:NiagaraPosition=<-10 0 0>)
    )
    Offset Amount=1
    Ribbon Width=181
   )
   &AddVelocity(Velocity=<-1000 0 0>, Velocity Speed Scale=2) {module=AddVelocity001}
   /SetVariables(Particles.Lines Ribbon Width=120) {isSetParameters=true}
  ]
  particleUpdateScript [
   &ParticleState()
   &ScaleRibbonWidth(Ribbon Width Scale=&FloatFromCurve(FloatCurve=DataInterface() {propertyValues {Curve {keys [
         {arriveTangent=2.5201, interpMode=RCIM_Cubic, leaveTangent=2.5201, tangentMode=RCTM_User, time=0, value=1}
         {arriveTangent=-0.5, interpMode=RCIM_Cubic, leaveTangent=-0.5, time=0.15, value=1}
         <at 1 0.5 RCIM_Cubic>
        ]}}}))
   &InheritSourceMovement(
    Applied Position Scale=&VectorFromFloat(Value=&FloatFromCurve(FloatCurve=DataInterface() {propertyValues {Curve {keys [<at 0 0.5 RCIM_Cubic>,<at 1 0 RCIM_Cubic>]}}}))
    ApplyForce=false
    ApplyPosition=true
    ApplyPositionFalloff=false
   )
   &SolveForcesAndVelocity(Clamp Velocity=false, Limit Acceleration=false)
   &VC_UpdateRibbon_v1()
   /SetVariables(
    Particles.Lines Ribbon Color:LinearColor=&VC_DynamicScaleColor(
     Color to Scale:LinearColor=<rgba 1 0.3215 0.25 1>
     RGB Scale:Vector3f=<1 1 1>
     Alpha Scale:NiagaraFloat=&FloatFromCurve(FloatCurve=DataInterface() {propertyValues {Curve {keys [<at 0.72 1>,<at 0.75 0>]}}})
    )
    Particles.SecondaryRibbonColor:LinearColor=<rgba 0.526 0.0164 0.0222 1>
   ) {isSetParameters=true}
  ]
  rendererClasses [&NiagaraRibbonRendererProperties]
  renderers [
   {rendererClass=&NiagaraRibbonRendererProperties}
   {rendererClass=&NiagaraRibbonRendererProperties, rendererIndex=1}
   {rendererClass=&NiagaraRibbonRendererProperties, rendererIndex=2}
  ]
 }
 Particles_Cone {
  emitterUpdateScript [
   &EmitterState()
   &SpawnRate(SpawnRate=12, Use Spawn Probability=false)
   &SpawnRate(SpawnGroup=1, SpawnRate=12, Use Spawn Probability=false) {module=SpawnRate001}
   &SpawnBurst_Instantaneous(Spawn Group=2, Use Spawn Probability=false, UseLoopCountLimit=false)
   &SpawnBurst_Instantaneous(Spawn Group=3, Use Spawn Probability=false, UseLoopCountLimit=false) {module=SpawnBurst_Instantaneous001}
   &SpawnBurst_Instantaneous(Spawn Group=4, Use Spawn Probability=false, UseLoopCountLimit=false) {module=SpawnBurst_Instantaneous002}
  ]
  particleSpawnScript [
   &VC_PaintedVariantsOverride() {enabled=false}
   /SetVariables(Particles.SetVizTag:NiagaraInt32=&SpawnGroup()) {isSetParameters=true}
   &InitializeParticle(
    Color Mode=::NewEnumerator0
    Lifetime=&VC_FloatBySpawnGroup(
     SpawnGroup or ExecIndex:NiagaraInt32=0
     CustomInt:NiagaraInt32=0
     MaximumSpawnGroups:NiagaraInt32=4
     Float SG 0:NiagaraFloat=0.1
     Float SG 1:NiagaraFloat=0.1
     Float SG 2:NiagaraFloat=$Emitter.CurrentLoopDuration
     Float SG 3:NiagaraFloat=$Emitter.CurrentLoopDuration
     Float SG 4:NiagaraFloat=0
     Float SG 5:NiagaraFloat=0
    )
    Mesh Scale=&RandomRangeVector(Minimum=<2.64 1.1 1.1>, Maximum=<2.42 0.88 0.88>, Random Seed=3115)
    Mesh Scale Mode=::NewEnumerator4
    Position Offset=&AddVectorToPosition(
     Position=$System.FinalPosition
     Vector=/VehicleCosmetics/FX/Niagara/Modules/MergeEmitters/VC_VectorBySpawnGroup(
      SpawnGroup or ExecIndex:NiagaraInt32=0
      CustomInt:NiagaraInt32=0
      MaximumSpawnGroups:NiagaraInt32=5
      Vector Group 0:Vector3f=<-67 0 0>
      Vector Group 1:Vector3f=<-53 0 0>
      Vector Group 2:Vector3f=<-55 0 0>
      Vector Group 3:Vector3f=<-85 0 0>
      Vector Group 4:Vector3f=<-180 0 0>
      Vector Group 5:Vector3f=<0 0 0>
     )
    )
    Position Offset Coordinate Space=::Local
    Sprite Rotation Angle=&MaskFloatBySpawnGroup(Masked Float=90, Passthrough Float=-89, Spawn Group Mask=4)
    Sprite Rotation Mode=::NewEnumerator0
    Sprite Size Mode=::NewEnumerator3
    Uniform Sprite Size=&VC_FloatBySpawnGroup(
     SpawnGroup or ExecIndex:NiagaraInt32=0
     CustomInt:NiagaraInt32=0
     MaximumSpawnGroups:NiagaraInt32=5
     Float SG 0:NiagaraFloat=0
     Float SG 1:NiagaraFloat=0
     Float SG 2:NiagaraFloat=300
     Float SG 3:NiagaraFloat=&RandomRangeFloat(Minimum=132, Maximum=120)
     Float SG 4:NiagaraFloat=190
     Float SG 5:NiagaraFloat=0
    )
    UsePositionOffset=true
    Write Material Random=true
   )
   &AddVelocity(Velocity=<-1500 0 0>) {enabled=false}
   &Color(
    Color=/VehicleCosmetics/FX/Niagara/Modules/MergeEmitters/VC_Dynamic_ColorBySpawnGroup(
     Color SG 0:LinearColor=<rgba 1.9 1.0122 0.209 1>
     MaximumSpawnGroups:NiagaraInt32=5
     Color SG 1:LinearColor=<rgba 1.9 1.0122 0.209 1>
     Color SG 2:LinearColor=<rgba 1.125 0.45 0.075 1>
     Color SG 3:LinearColor=<rgba 0.75 0.5545 0.1914 1>
     Color SG 4:LinearColor=<rgba 0.9531 0.4378 0.3177 1>
    )
    Scale Alpha=&MaskFloatBySpawnGroup(Masked Float=0.4, Passthrough Float=1, Spawn Group Mask=2)
   )
   &InitialMeshOrientation(
    Mesh Orientation Mode=::NewEnumerator7
    Rotation=&RandomRangeVector(Minimum=<-1 0 0>, Maximum=<1 0 0>)
    Rotation Coordinate Space=::NewEnumerator0
    Use Rotation Vector=true
   )
   /SetVariables(Particles.SizeFlame:Vector2f=<130 120>, Particles.SizeFlame2:Vector2f=<240 180>) {isSetParameters=true}
  ]
  particleUpdateScript [
   &ParticleState()
   &ScaleMeshSize(Scale Factor=&MaskVectorBySpawnGroup(Masked Vector=<0.429 0.462 0.462>, Passthrough Vector=<0.264 0.561 0.561>))
   &SolveForcesAndVelocity(Clamp Velocity=false, Limit Acceleration=false)
   &ScaleColor(
    Scale Alpha=&MaskFloatBySpawnGroup(
     Masked Float=&FloatFromCurve(FloatCurve=DataInterface() {propertyValues {Curve {keys [<at 0 1>,<at 6 0>]}}})
     Passthrough Float=1
     Spawn Group Mask=3
    )
    Scale RGB=&MaskVectorBySpawnGroup(Masked Vector=<1 1 1>, Passthrough Vector=<1.3 1.3 1.3>, Spawn Group Mask=4)
    ScaleA=true
    ScaleRGB=true
   )
   &SubUVAnimation(
    Output Alternate Renderer Binding=true
    "Sprite Renderer "=DataInterface() {propertyValues {SpriteRenderer=&self:NiagaraEmitter_1.NiagaraSpriteRendererProperties_1}}
    UseEndFrame=false
    UseStartFrame=false
   )
   &CameraOffset(Camera Offset Amount=&VC_FloatBySpawnGroup(
     SpawnGroup or ExecIndex:NiagaraInt32=0
     CustomInt:NiagaraInt32=0
     MaximumSpawnGroups:NiagaraInt32=5
     Float SG 0:NiagaraFloat=0
     Float SG 1:NiagaraFloat=0
     Float SG 2:NiagaraFloat=0
     Float SG 3:NiagaraFloat=20
     Float SG 4:NiagaraFloat=7
     Float SG 5:NiagaraFloat=0
    ))
  ]
  rendererClasses [&NiagaraSpriteRendererProperties, &NiagaraMeshRendererProperties]
  renderers [
   {rendererClass=&NiagaraSpriteRendererProperties}
   {rendererClass=&NiagaraMeshRendererProperties, rendererIndex=1}
   {rendererClass=&NiagaraMeshRendererProperties, rendererIndex=2}
   {rendererClass=&NiagaraSpriteRendererProperties, rendererIndex=3}
   {rendererClass=&NiagaraSpriteRendererProperties, rendererIndex=4}
  ]
 }
}
summary {
 systemName=NS_Boost_Gray
 emitters [
  {emitterName=Dots, rendererClasses [&NiagaraSpriteRendererProperties]}
  {emitterName=Particles_Cone, rendererClasses [&NiagaraSpriteRendererProperties, &NiagaraMeshRendererProperties]}
  {emitterName=BlackLightningRibbon001, rendererClasses [&NiagaraRibbonRendererProperties]}
  {emitterName=MainTrail001, rendererClasses [&NiagaraRibbonRendererProperties]}
 ]
 userVariables [
  User.bIsFrontendPreview {type=NiagaraBool}
  User.bIsPainted {type=NiagaraBool}
  User.ColorVariantID {type=NiagaraInt32}
  User.DTCarOffset {type=NiagaraPosition, description="", defaultValue {}}
  User.PaintedColor {type=LinearColor, description="", defaultValue {struct=LinearColor, value=<rgba 1 1 1 1>}}
  User.PaintedScalar {type=NiagaraFloat, defaultValue=1, description=""}
  User.PortCount {type=NiagaraInt32, defaultValue=1}
  User.SingleBoostPos {type=NiagaraPosition, description="", defaultValue {}}
 ]
}
